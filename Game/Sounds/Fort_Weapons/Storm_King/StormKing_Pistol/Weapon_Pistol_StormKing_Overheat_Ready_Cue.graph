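# Sound cue for the Storm King pistol, apparently played when the weapon is ready again after overheating.
# A mixer layers a random Storm King assault overheat-ready sample, pitched well up, with a random pistol overheat-release sample, pitched down and faded out by a volume envelope.
# It uses weapon foley class and reload attenuation, overrides concurrency, and has its debug flag left on.
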SoundCue /Game/Sounds/Fort_Weapons/Storm_King/StormKing_Pistol/Weapon_Pistol_StormKing_Overheat_Ready_Cue
root=Mixer_0
nodes [
 {
  id=Enveloper_0
  children [Modulator_1]
  properties {pitchCurve {editorCurveData {keys [{time=0}]}}, volumeCurve {editorCurveData {keys [<at 0 1 RCIM_Cubic>,<at 3 0 RCIM_Cubic>]}}}
 }
 {id=Mixer_0, children [Modulator_0,Enveloper_0], properties {inputVolume [1.1,1]}}
 {id=Modulator_0, children [Random_0], properties {pitchMax=2.5, pitchMin=2}}
 {id=Modulator_1, children [Random_2], properties {pitchMax=0.8, pitchMin=0.7}}
 {id=Random_0, children [WavePlayer_0,WavePlayer_1,WavePlayer_2], properties {weights [1,1,1]}}
 {id=Random_2, children [WavePlayer_9, WavePlayer_10, WavePlayer_11], properties {weights [1,1,1]}}
 {id=WavePlayer_0, properties {soundWaveAssetPtr=/Game/Sounds/Fort_Weapons/Storm_King/StormKing_Assault/StormKing_Assault_Overheat_Ready_01}}
 {id=WavePlayer_1, properties {soundWaveAssetPtr=/Game/Sounds/Fort_Weapons/Storm_King/StormKing_Assault/StormKing_Assault_Overheat_Ready_02}}
 {id=WavePlayer_10, properties {soundWaveAssetPtr=&/StormKing_Pistol_Overheat_Release_02}}
 {id=WavePlayer_11, properties {soundWaveAssetPtr=&/StormKing_Pistol_Overheat_Release_03}}
 {id=WavePlayer_2, properties {soundWaveAssetPtr=/Game/Sounds/Fort_Weapons/Storm_King/StormKing_Assault/StormKing_Assault_Overheat_Ready_03}}
 {id=WavePlayer_9, properties {soundWaveAssetPtr=&/StormKing_Pistol_Overheat_Release_01}}
]
properties {
 attenuationSettings=&Fort_Weapon_Reload_Att
 bDebug=true
 bOverrideConcurrency=true
 maxDistance=2250
 soundClassObject=&Fort_WeaponFoley
 volumeMultiplier=0.451
}
